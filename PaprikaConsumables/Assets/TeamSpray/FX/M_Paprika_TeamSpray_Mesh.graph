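# Translucent, two-sided subsurface material for the Team Spray effect's Niagara mesh particles, resembling a soft, dissolving volume.
# Animated Voronoi volume noise drives normal displacement and blends two colors, split between base color and emissive.
# A camera offset, skipped in shadows, shifts faces pointing away from view.
# Opacity combines panning dissolve noise, particle alpha, inverse Fresnel and camera/depth fades, with dynamic parameters controlling displacement, dissolve, noise and luminance.
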Material /PaprikaConsumables/Assets/TeamSpray/FX/M_Paprika_TeamSpray_Mesh
graph {
 properties {
  bUsedWithNiagaraMeshParticles=true
  blendMode=BLEND_Translucent
  pixelDepthOffsetMode=PDOM_Legacy
  shadingModel=MSM_Subsurface
  twoSided=true
  usedShadingModels=Subsurface
 }
 nodes [
  MaterialFunctionCall(Emissive=NamedRerouteUsage()) {materialFunction=&MF_FX_EmissiveAdjustments}
  Multiply(A=NamedRerouteUsage()) {constB=0.6}
  Multiply() {constA=1}
  n1 = DynamicParameter() {parameterIndex=1, defaultValue=<rgba 5 0.008 0.5 1>, paramNames [Temperature Pow, Noise Scale, Luminance Adjustment, Fresnel Mul]}
  n2 = Multiply(A=VectorParameter() {parameterName=Volonoi Noise Scale, defaultValue=<rgba 1 1 1 0>}.RGB, B=@n1.Noise Scale)
  n3 = Multiply(
   A=DynamicParameter() {defaultValue=<rgba -100 1 1 -1>, paramNames [Displacement, ShadowStrength, PlayAnimTime, Dissolve]}.PlayAnimTime
   B=Constant3Vector() {constant=<rgba 0 0 1 0>}
  )
  n4 = Multiply(A=ParticleRandom()) {constB=100}
  n5 = Subtract(A=Add(A=Add(A=@n3, B=@n4), B=WorldPosition().XYZ), B=Add(A=ParticlePositionWS(), B=ParticleRandom()))
  n6 = TextureSample(UVs=Multiply(A=Multiply(A=@n2, B=@n5)) {constB=0.5}, Tex=NamedRerouteUsage()) {samplerType=SAMPLERTYPE_LinearColor, texture=/Game/Effects/Explosion_Update/Textures/VT_Voronoi_Seamless}
  n7 = Multiply(A=@n6.R, B=@n6.R)
  NamedRerouteDeclaration(None=Multiply(A=@n7, B=@n7)) {name=Voronoi Noise Result 2, nodeColor=<rgba 1 0 0.0703 1>}
  n8 = Multiply(A=VertexNormalWS(), B=NamedRerouteUsage())
  n9 = DynamicParameter() {defaultValue=<rgba -100 1 1 -1>, paramNames [Displacement, ShadowStrength, PlayAnimTime, Dissolve]}
  n10 = Subtract(A=WorldPosition() {worldPositionShaderOffset=WPT_ExcludeAllShaderOffsets}.XYZ, B=CameraPositionWS())
  n11 = Saturate(None=DotProduct(A=VertexNormalWS(), B=Normalize(VectorInput=@n10)))
  n12 = ScalarParameter() {defaultValue=1, parameterName=WPO Inner Exp}
  n13 = MaterialFunctionCall(Offset Amount=Multiply(A=Power(Base=@n11, Exp=@n12), B=ScalarParameter() {defaultValue=150, parameterName=WPO Inner Value})) {materialFunction=/Engine/Functions/Engine_MaterialFunctions02/WorldPositionOffset/CameraOffset}
  n14 = Constant3Vector()
  NamedRerouteDeclaration(None=Add(A=Multiply(A=@n8, B=@n9.Displacement), B=ShadowReplace(Default=@n13.Clamped Camera Offset, Shadow=@n14)) {constB=0}) {name=Displacement Result, nodeColor=<rgba 0 0.6484 1 1>}
  n15 = VectorParameter() {parameterName=Color B, defaultValue=<rgba 0.1 0.1 0.1 1>}
  n16 = VectorParameter() {parameterName=Color A, defaultValue=<rgba 1 1 1 1>}
  n17 = TextureCoordinate()
  n18 = Saturate(None=Multiply(A=ComponentMask(None=@n17) {g=true}) {constB=2})
  n19 = NamedRerouteUsage()
  n20 = LinearInterpolate(B=Power(Base=OneMinus(None=@n18)) {constExponent=1.5}, Alpha=Power(Base=Saturate(None=@n19)) {constExponent=1.5})
  n21 = DynamicParameter() {parameterIndex=1, defaultValue=<rgba 5 0.008 0.5 1>, paramNames [Temperature Pow, Noise Scale, Luminance Adjustment, Fresnel Mul]}
  n22 = Multiply(
   A=Multiply(A=LinearInterpolate(A=@n15.RGB, B=@n16.RGB, Alpha=@n20), B=ParticleColor().RGB)
   B=Multiply(A=@n21.Luminance Adjustment, B=Constant() {r=1.5})
  )
  n23 = ScalarParameter() {defaultValue=0.5, parameterName=Diffuse Emissive Ratio}
  NamedRerouteDeclaration(None=LinearInterpolate(A=@n22, Alpha=@n23) {constB=0}) {name=Emissive Result, nodeColor=<rgba 0 0.5781 1 1>}
  NamedRerouteDeclaration(None=Multiply(A=@n22, B=@n23)) {name=Base Color Result, nodeColor=<rgba 0.0312 1 0 1>}
  n24 = TextureObjectParameter() {
   group=Volume Textures
   parameterName=Dissolve and Swirl Noise Volume Tex
   samplerType=SAMPLERTYPE_LinearColor
   texture=/Game/Global/MaterialFunctions/VT_ClothSimWPO
  }
  NamedRerouteDeclaration(None=@n24) {name=Noise Texture 1, nodeColor=<rgba 0 0.6719 1 1>}
  n25 = TextureObjectParameter() {
   group=Volume Textures
   parameterName=Displacement Volume Tex
   samplerType=SAMPLERTYPE_LinearColor
   texture=&VT_Voronoi_Seamless
  }
  NamedRerouteDeclaration(None=@n25) {name=Noise Texture 2, nodeColor=<rgba 0 0.7188 1 1>}
  n26 = TextureCoordinate()
  n27 = Multiply(A=OneMinus(None=ComponentMask(None=@n26) {g=true}), B=ScalarParameter() {defaultValue=1.33, parameterName=Fade To Bottom Multiplier})
  NamedRerouteDeclaration(None=Power(Base=Saturate(None=@n27), Exp=ScalarParameter() {defaultValue=4, parameterName=Fade To Bottom Exp})) {name=Fade to Bottom, nodeColor=<rgba 0 1 0.9063 1>}
  NamedRerouteDeclaration(None=LinearInterpolate(A=@n15.A, B=@n16.A, Alpha=@n20)) {name=Color Opacity, nodeColor=<rgba 1 0 0.9375 1>}
  NamedRerouteUsage()
  n28 = MaterialFunctionCall(Fade Length=ScalarParameter() {defaultValue=300, parameterName=Near Camera Fade Distance}) {materialFunction=&CameraDepthFade}
  n29 = DynamicParameter() {parameterIndex=3, defaultValue=<rgba 0 0 1 0>, paramNames [Fire Pan, Dissolve Pan, Temperature, FireNoise Pan]}
  n30 = Multiply(A=Constant3Vector() {constant=<rgba 0 0 1 0>}, B=@n29.Dissolve Pan)
  n31 = Time()
  n32 = VectorParameter() {group=DIssolve, parameterName=Dissolve Noise Pan Speed XYZ, defaultValue=<rgba 1 1 -40 0>}
  n33 = Multiply(
   A=Add(A=WorldPosition().XYZ, B=Multiply(A=@n31, B=@n32.RGB))
   B=Multiply(
    A=VectorParameter() {group=DIssolve, parameterName=DissolveNoiseScale, defaultValue=<rgba 0.002 0.002 0.005 1>}.RGB
    B=ScalarParameter() {defaultValue=0.6, parameterName=Dissolve Noise Scale}
   )
  ) {constB=0.01}
  n34 = OneMinus(None=TextureSample(UVs=Add(A=@n30, B=@n33), Tex=NamedRerouteUsage()) {samplerType=SAMPLERTYPE_LinearColor}.R)
  n35 = Saturate(None=Subtract(
    A=Multiply(A=@n34, B=@n34)
    B=DynamicParameter() {defaultValue=<rgba -100 1 1 -1>, paramNames [Displacement, ShadowStrength, PlayAnimTime, Dissolve]}.Dissolve
   ))
  n36 = Multiply(
   A=Multiply(A=NamedRerouteDeclaration(None=@n35) {name=Dissolve Noise, nodeColor=<rgba 0.6875 1 0 1>}, B=ParticleColor().A)
   B=NamedRerouteUsage()
  )
  n37 = NamedRerouteUsage()
  n38 = OneMinus(None=Fresnel(
    ExponentIn=ScalarParameter() {defaultValue=1, parameterName=Fresnel Alpha Expo}
    BaseReflectFractionIn=ScalarParameter() {parameterName=BaseReflectFraction}
   ) {baseReflectFraction=0.01})
  n39 = DynamicParameter() {parameterIndex=1, defaultValue=<rgba 5 0.008 0.5 1>, paramNames [Temperature Pow, Noise Scale, Luminance Adjustment, Fresnel Mul]}
  n40 = Multiply(A=ScalarParameter() {defaultValue=6, parameterName=FresnelPower}, B=@n39.Fresnel Mul)
  n41 = Subtract(A=Multiply(A=Power(Base=@n38, Exp=@n40)) {constB=2}) {constB=0.1}
  n42 = Multiply(A=Multiply(A=@n28.Result, B=@n28.Result), B=Multiply(A=Multiply(A=@n36, B=@n37), B=Saturate(None=@n41)))
  n43 = DepthFade(Opacity=@n42, FadeDistance=ScalarParameter() {defaultValue=500, parameterName=DepthFade})
  Reroute(None=Saturate(None=ShadingPathSwitch(Default=@n43, Deferred=@n43, Mobile=@n42) {desc=Migrated from FeatureLevelSwitch}))
 ]
}
