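# Niagara area-of-effect visual for the Paprika TeamSpray consumable, combining foam, soda bubbles and a boundary mesh with decal.
# GPU Foam_Umbrella sprites spawn on a sphere band swept around by emitter age, with tangential and downward velocity plus gravity.
# Soda_pop sprites fill the spherical volume, swirl under curl noise and are distance-culled.
# Sphere_with_Decal bursts a mesh and decal, animates material parameters by system age, and fades through an execution-state-driven timeline.
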NiagaraSystem /PaprikaConsumables/Assets/TeamSpray/FX/NS_Paprika_TeamSpray_AOE
dependencies {
 usedDataInterfaces [&NiagaraDataInterfaceCurve, &NiagaraDataInterfaceColorCurve, &NiagaraDataInterfaceSkeletalMesh]
 usedDynamicInputs []
 usedModules [
  &self:SystemSpawnScript.SystemScriptSource.SystemScriptGraph.NiagaraNodeAssignment_0.SetVariables_#1
  &self:SystemSpawnScript.SystemScriptSource.SystemScriptGraph.NiagaraNodeAssignment_1.SetVariables_#2
  &self:SystemSpawnScript.SystemScriptSource.SystemScriptGraph.NiagaraNodeAssignment_3.SetVariables_#3
  &self:SystemSpawnScript.SystemScriptSource.SystemScriptGraph.NiagaraNodeAssignment_2.SetVariables_#4
  &SystemState
 ]
}
emitters {
 Foam_Umbrella {
  simTarget=GPUComputeSim
  emitterUpdateScript [
   &EmitterState()
   &SpawnRate(SpawnRate=120, Use Spawn Probability=false)
  ]
  particleSpawnScript [
   &InitializeParticle(
    Lifetime=0.75
    Position Offset=$System.PositionOffset
    Sprite Rotation Mode=::NewEnumerator2
    Sprite Size Mode=::NewEnumerator1
    Sprite UV Mode=::NewEnumerator1
    Uniform Sprite Size Max=5
    Uniform Sprite Size Min=3
    UsePositionOffset=false
    Write Material Random=true
   )
   &SubUVAnimation(End Frame=3, SubUV Animation Mode:ENiagara_SubUVLookupMode=::NewEnumerator2)
   &ShapeLocation(
    MaskSpawnGroup=false
    Rotation Axis=<0 0 1>
    Rotation Mode=::NewEnumerator1
    Sphere Distribution=::NewEnumerator1
    Sphere Radius=&RandomRangeFloat(Minimum=450, Maximum=460)
    U Position=&Multiply_Float(A=$Emitter.Age, B=&RandomRangeFloat(Minimum=15, Maximum=17))
    UseOwnerScale=true
    V Position=&RandomRangeFloat(Minimum=0.9, Maximum=1)
   )
   &AddVelocity(Velocity=$StackContext.ShapeLocation.ShapeTangent, Velocity Speed Scale=-150) {module=AddVelocity001}
   &AddVelocity(Velocity=<0 0 -1>, Velocity Speed Scale=100)
  ]
  particleUpdateScript [
   &ParticleState()
   &GravityForce(Gravity=&Multiply_VectorByFloat(Vector=<0 0 -980>, Float=1))
   &Drag(
    Drag=&RandomRangeFloat(Minimum=0.5, Maximum=1)
    Rotational Drag=50
    Use Linear Drag=true
    Use Rotational Drag=false
   ) {enabled=false}
   &InheritSourceMovement(
    Applied Position Scale=&VectorFromFloat(Value=$System.Inherit Source Movement)
    ApplyForce=false
    ApplyPosition=true
    ApplyPositionFalloff=false
   )
   &SolveForcesAndVelocity(Clamp Velocity=false, Limit Acceleration=false)
   &SpriteRotationRate(Rotation Rate=&FloatFromCurve(
     FloatCurve=DataInterface() {propertyValues {Curve {keys [<at 0 1>,<at 1 0>]}}}
     Scale Curve=&MakeCustomFloatFromBool(Bool=&RandomBool(), True Float=-90, False Float=90)
    ))
   &ScaleSpriteSize(Scale Factor=&Vector2DFromFloat(Value=&FloatFromCurve(FloatCurve=DataInterface() {propertyValues {Curve {keys [
          {arriveTangent=3.8162, interpMode=RCIM_Cubic, leaveTangent=3.8162, tangentMode=RCTM_User, time=0, value=0}
          {interpMode=RCIM_Cubic, tangentMode=RCTM_User, time=0.33, value=0.975}
          {arriveTangent=-0.0405, interpMode=RCIM_Cubic, leaveTangent=-0.0405, tangentMode=RCTM_User, time=0.7009, value=1}
          {arriveTangent=-8.0089, interpMode=RCIM_Cubic, leaveTangent=-8.0089, tangentMode=RCTM_User, time=1, value=0}
         ]}}})))
   &ScaleColor(
    Scale Alpha=&FloatFromCurve(FloatCurve=DataInterface() {propertyValues {Curve {keys [
         <at 0 0 RCIM_Cubic>
         {interpMode=RCIM_Cubic, tangentMode=RCTM_User, time=0.2, value=1}
         {interpMode=RCIM_Cubic, tangentMode=RCTM_User, time=0.8, value=1}
         {interpMode=RCIM_Cubic, tangentMode=RCTM_User, time=1, value=0}
        ]}}})
    Scale RGB=&Lerp_Vector(
     A=<0 0.5 2>
     B=<0.125 0 0.6>
     Alpha=&ClampFloat(Float=&FloatFromCurve(
       FloatCurve=DataInterface() {propertyValues {Curve {keys [{tangentMode=RCTM_User, time=0, value=0}, {tangentMode=RCTM_User, time=0.75, value=1}]}}}
       Scale Curve=&RandomRangeFloat(Minimum=1, Maximum=3)
      ))
    )
    ScaleA=true
    ScaleRGB=true
   )
   &DynamicMaterialParameters(
    Index 0 Param 2=&FloatFromCurve(
     FloatCurve=DataInterface() {propertyValues {Curve {keys [{arriveTangent=-1.9579, interpMode=RCIM_Cubic, time=0, value=1}, {interpMode=RCIM_Cubic, tangentMode=RCTM_User, time=0.5, value=0}]}}}
     Scale Curve=0
    )
    Index 0 Param 4=&FloatFromCurve(FloatCurve=DataInterface() {propertyValues {Curve {keys [{arriveTangent=1.1512, interpMode=RCIM_Cubic, leaveTangent=1.1512, tangentMode=RCTM_User, time=0, value=0}, <at 1 1 RCIM_Cubic>]}}})
    Module.Param0WriteEnabled=false
    Module.Param1WriteEnabled=true
    Module.Param2WriteEnabled=false
    Module.Param3WriteEnabled=false
   )
   &ScaleSpriteSizeBySpeed(Max Scale Factor=<0.6 2>) {enabled=false}
  ]
  rendererClasses [&NiagaraSpriteRendererProperties]
  renderers [{rendererClass=&NiagaraSpriteRendererProperties}]
 }
 Soda_pop {
  emitterUpdateScript [
   &EmitterState(Enable Distance Culling=true, MaxDistance=2250, Module.bUseMaxDistance=true, Module.bUseMinDistance=false, Scalability Mode=::NewEnumerator1)
   &SpawnRate(SpawnRate=120, Use Spawn Probability=false)
  ]
  particleSpawnScript [
   &InitializeParticle(
    Color=&UniformRangedLinearColor(Minimum=<rgba 0.5 0 2 1>, Maximum=<rgba 0 1 4 1>)
    Lifetime=&UniformRangedFloat(Minimum=0.18, Maximum=1)
    Mass Mode=::NewEnumerator1
    Sprite Size=&Vector2DFromFloat(Value=&UniformRangedFloat(Minimum=4, Maximum=12))
    Sprite Size Mode=::NewEnumerator4
    UsePositionOffset=false
    Write Material Random=true
   )
   &ShapeLocation(MaskSpawnGroup=false, Override Seed=false, Sphere Radius=500, UseOwnerScale=true)
   &AddVelocity(Velocity=<0 0 1>, Velocity Speed Scale=50)
  ]
  particleUpdateScript [
   &ParticleState()
   /Niagara/Modules/Update/Forces/CurlNoiseForce(Noise Strength=800, Pan Noise=false)
   &SolveForcesAndVelocity(Clamp Velocity=false, Limit Acceleration=false)
   &ScaleSpriteSize(Scale Factor=&Vector2DFromFloat(Value=&FloatFromCurve(FloatCurve=DataInterface() {propertyValues {Curve {keys [{arriveTangent=3.6722, interpMode=RCIM_Cubic, leaveTangent=3.6722, tangentMode=RCTM_User, time=0, value=0}, <at 0.8 1>, <at 1 2>]}}})))
   &ScaleColor(
    Scale Alpha=&FloatFromCurve(FloatCurve=DataInterface() {propertyValues {Curve {keys [
         <at 0 1>
         {arriveTangent=-4.4715, interpMode=RCIM_Cubic, leaveTangent=-4.4715, tangentMode=RCTM_User, time=0.8022, value=0.9932}
         {arriveTangent=-7.8513, interpMode=RCIM_Cubic, leaveTangent=-7.8513, tangentMode=RCTM_User, time=1, value=0}
        ]}}})
    ScaleA=true
    ScaleRGB=true
   )
  ]
  rendererClasses [&NiagaraSpriteRendererProperties]
  renderers [{rendererClass=&NiagaraSpriteRendererProperties}]
 }
 Sphere_with_Decal {
  emitterSpawnScript [
   /SetVariables(Emitter.Global Color:LinearColor=<rgba 0 1 4 1>, Emitter.Global Scale=1) {isSetParameters=true}
   /SetVariables(
    Emitter.Color2:LinearColor=&MultiplyLinearColor(Linear Color=$Emitter.Global Color, Scale Factor=<2 1 1 1>)
    Emitter.Depth Fade:NiagaraFloat=&Multiply_Float(A=125, B=$Emitter.Global Scale)
    Emitter.Dissolve Noise Power=1.1
    Emitter.Dissolve Noise Scale:LinearColor=&MultiplyLinearColor(
     Linear Color=/Niagara/DynamicInputs/TypeConversions/MakeLinearColorFromFloat(Value:NiagaraFloat=&FloatFromCurve(FloatCurve=DataInterface() {propertyValues {Curve {keys [<at 0 2>,<at 1 1>,<at 2 0.5>]}}}, CurveIndex=$Emitter.Global Scale))
     Scale Factor=<0.0008 0.0008 0.0009 1>
    )
    Emitter.Near Camera Fade:NiagaraFloat=&Multiply_Float(A=400, B=$Emitter.Global Scale)
    Emitter.Use SkyAtmosphereLightColor=true
   ) {enabled=false, isSetParameters=true}
  ]
  emitterUpdateScript [
   &EmitterState()
   &SpawnBurst_Instantaneous(Spawn Group=1, Spawn Time=0.25, Use Spawn Probability=false, UseLoopCountLimit=true) {module=SpawnBurst_Instantaneous002}
  ]
  particleSpawnScript [
   &InitializeParticle(
    Mesh Scale Mode=::NewEnumerator3
    Mesh Uniform Scale=&Divide_Float(A=1024, B=174)
    UsePositionOffset=false
    Write Material Random=true
   )
   /SetVariables(Particles.Scale=&Multiply_VectorByFloat(Vector=$Particles.Scale, Float=$Emitter.Global Scale)) {isSetParameters=true}
   /SetVariables(Particles.DecalSize:Vector3f=&Multiply_VectorByFloat(Vector=<1100 1100 1100>, Float=$Emitter.Global Scale)) {isSetParameters=true}
  ]
  particleUpdateScript [
   &ParticleState(Kill Particles When Lifetime Has Elapsed=false)
   &Timeline(Initial Start Time=0.0333, Play=/Game/Effects/Niagara/DynamicInputs/State/FN_IsExecutionStateActive())
   &KillParticles(Kill Particles=&SetBoolByFloatComparison(A=$StackContext.Timeline.NormalizedTime, B=0, Comparison Type=::NewEnumerator4))
   /SetVariables(Particles.Color Decal:LinearColor=&MakeLinearColorFromVectorAndFloat("Vector (RGB)"=<1 1 1>, "Float (Alpha)"=$Particles.Timeline.NormalizedTime)) {isSetParameters=true}
   &Drag(Drag=&RandomRangeFloat(Minimum=9, Maximum=10), Use Linear Drag=true, Use Rotational Drag=false)
   &DynamicMaterialParameters(
    Index 0 Param 1=-100
    Index 0 Param 3=&Multiply_Float(A=$System.Age, B=600)
    Index 1 Param 1=1
    Index 1 Param 2=0.003
    Index 1 Param 3=2
    Index 1 Param 4=0.5
    Index 3 Param 1=&Multiply_Float(A=$System.Age, B=1)
    Index 3 Param 2=&Multiply_Float(A=$System.Age, B=0.5)
    Index 3 Param 3=0.1
    Index 3 Param 4=&Multiply_Float(A=$System.Age, B=1)
    Module.Index0Param0Write=true
    Module.Index0Param1Write=true
    Module.Index0Param2Write=true
    Module.Index0Param3Write=true
    Module.Index1Param0Write=true
    Module.Index1Param1Write=true
    Module.Index1Param2Write=true
    Module.Index1Param3Write=true
    Module.Index3Param0Write=false
    Module.Index3Param1Write=true
    Module.Index3Param2Write=true
    Module.Index3Param3Write=true
    Write Parameter Index 1=true
    Write Parameter Index 3=true
   ) {module=DynamicMaterialParameters001}
   &ScaleColor(Scale Alpha=&Multiply_Float(A=1, B=$Particles.Timeline.NormalizedTime), ScaleA=true, ScaleRGB=true)
   &SolveForcesAndVelocity(Clamp Velocity=false, Limit Acceleration=false)
   &ScaleMeshSize()
  ]
  rendererClasses [&NiagaraMeshRendererProperties, &NiagaraDecalRendererProperties]
  renderers [
   {rendererClass=&NiagaraMeshRendererProperties, rendererIndex=0}
   {rendererClass=&NiagaraDecalRendererProperties, rendererIndex=1}
  ]
 }
}
summary {
 systemName=NS_Paprika_TeamSpray_AOE
 emitters [
  {emitterName=Soda_pop, rendererClasses [&NiagaraSpriteRendererProperties]}
  {emitterName=Foam_Umbrella, simTarget=GPUComputeSim, rendererClasses [&NiagaraSpriteRendererProperties]}
  {
   emitterName=Sphere_with_Decal
   rendererClasses [&NiagaraMeshRendererProperties, &NiagaraDecalRendererProperties]
  }
 ]
 userVariables [
  User.bisInWater {type=NiagaraBool}
  User.Global Scale {type=NiagaraFloat, defaultValue=1, description=""}
  User.WaterBodyIndex {type=NiagaraFloat, defaultValue=0, description=""}
  User.WaterDepth {type=NiagaraFloat, defaultValue=0, description=""}
  User.WaterPlaneLocation {type=NiagaraPosition, description="", defaultValue {}}
  User.WaterPlaneNormal {type=Vector3f, description="", defaultValue {struct=Vector3f, value=<0 0 1>}}
  User.WaveAttenuationFactor {type=NiagaraFloat, defaultValue=0, description=""}
  User.WaveTime {type=NiagaraFloat, defaultValue=0, description=""}
 ]
}
