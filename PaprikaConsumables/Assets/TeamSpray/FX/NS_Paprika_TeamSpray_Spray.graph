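# Niagara system for the TeamSpray consumable's foam spray, blending two system colors and scaled by a shared System.Spray value.
# Foam_Spray emits gravity-affected sprites by rate and distance moved, while Dripping_Foam adds short-lived drips, both inheriting source movement.
# Foam_Spray_Directional repositions burst particles every update into a ribbon curving from owner up-axis toward SprayDirection, using normalized execution index.
# A sprite directional variant and a mesh emitter are disabled.
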NiagaraSystem /PaprikaConsumables/Assets/TeamSpray/FX/NS_Paprika_TeamSpray_Spray
dependencies {
 usedDataInterfaces [&NiagaraDataInterfaceCurve, &NiagaraDataInterfaceArrayFloat]
 usedDynamicInputs [
  &OneMinusFloat
  &Add_Vector
  /NiagaraFluids/DynamicInput/MinVectorByLength
  &Boolean_NotOperation
 ]
 usedModules [
  &self:SystemSpawnScript.SystemScriptSource.SystemScriptGraph.NiagaraNodeAssignment_1.SetVariables_#1
  &self:SystemSpawnScript.SystemScriptSource.SystemScriptGraph.NiagaraNodeAssignment_5.SetVariables_#2
  &self:SystemSpawnScript.SystemScriptSource.SystemScriptGraph.NiagaraNodeAssignment_6.SetVariables_#3
  &SystemState
  &Timeline
  &self:SystemSpawnScript.SystemScriptSource.SystemScriptGraph.NiagaraNodeAssignment_0.SetVariables_#4
  &self:SystemSpawnScript.SystemScriptSource.SystemScriptGraph.NiagaraNodeAssignment_2.SetVariables_#5
  &self:SystemSpawnScript.SystemScriptSource.SystemScriptGraph.NiagaraNodeAssignment_3.SetVariables_#6
  &self:SystemSpawnScript.SystemScriptSource.SystemScriptGraph.NiagaraNodeAssignment_7.SetVariables_#7
  &InterpolateOverTime
  &self:SystemSpawnScript.SystemScriptSource.SystemScriptGraph.NiagaraNodeAssignment_4.SetVariables_#8
  &self:ScratchModule_04
 ]
}
emitters {
 Dripping_Foam {
  emitterUpdateScript [
   &EmitterState()
   &SpawnRate(Spawn Probability=$System.Spray, SpawnRate=20, Use Spawn Probability=true)
  ]
  particleSpawnScript [
   &InitializeParticle(
    Lifetime Max=0.5
    Lifetime Min=0.4
    Lifetime Mode=::NewEnumerator1
    Sprite Rotation Mode=::NewEnumerator2
    Sprite Size Mode=::NewEnumerator1
    Sprite UV Mode=::NewEnumerator1
    Uniform Sprite Size Max=75
    Uniform Sprite Size Min=60
    UsePositionOffset=false
    Write Material Random=true
   )
   &ShapeLocation(
    MaskSpawnGroup=false
    Sphere Distribution=::NewEnumerator1
    Sphere Radius=5
    U Position=&Multiply_Float(A=$Emitter.Age, B=2)
    UseOwnerScale=true
    V Position=0.99
   )
   &AddVelocity(
    Cone Angle=2
    Cone Axis=<0 0 1>
    Module.UseConeFalloff=false
    Module.UseVelDistribution=false
    Override Seed=false
    Velocity Mode=::NewEnumerator2
    Velocity Speed=&Multiply_Float(A=20, B=$System.Spray)
   ) {module=AddVelocity001}
   /SetVariables(Particles.SubImageIndex=&RandomRangeFloat(Minimum=0, Maximum=3)) {isSetParameters=true}
  ]
  particleUpdateScript [
   &ParticleState()
   &CameraOffset(Camera Offset Amount=&Multiply_Float(A=&MakeFloatFromVector2d(Vector 2D=$Particles.SpriteSize), B=0.33))
   &GravityForce(Gravity=&Multiply_VectorByFloat(Vector=<0 0 -980>, Float=0.75))
   &Drag(
    Drag=&RandomRangeFloat(Minimum=0.5, Maximum=1)
    Rotational Drag=50
    Use Linear Drag=true
    Use Rotational Drag=false
   )
   &InheritSourceMovement(
    Applied Position Scale=&VectorFromFloat(Value=&FloatFromCurve(
      FloatCurve=DataInterface() {propertyValues {Curve {keys [{interpMode=RCIM_Cubic, tangentMode=RCTM_User, time=0.2, value=1}, {interpMode=RCIM_Cubic, tangentMode=RCTM_User, time=0.66, value=0}]}}}
      Scale Curve=$System.Inherit Source Movement
     ))
    ApplyForce=false
    ApplyPosition=true
    ApplyPositionFalloff=false
   )
   &SolveForcesAndVelocity(Clamp Velocity=false, Limit Acceleration=false)
   &SpriteRotationRate(Rotation Rate=&FloatFromCurve(
     FloatCurve=DataInterface() {propertyValues {Curve {keys [<at 0 1>,<at 1 0>]}}}
     Scale Curve=&MakeCustomFloatFromBool(Bool=&RandomBool(), True Float=-90, False Float=90)
    ))
   &ScaleSpriteSize(Scale Factor=&Vector2DFromFloat(Value=&FloatFromCurve(FloatCurve=DataInterface() {propertyValues {Curve {keys [
          {arriveTangent=2.3205, interpMode=RCIM_Cubic, leaveTangent=2.3205, tangentMode=RCTM_User, time=0, value=0.15}
          {interpMode=RCIM_Cubic, leaveTangent=0.4735, time=1, value=1}
         ]}}})))
   &ScaleColor(
    Scale Alpha=&FloatFromCurve(FloatCurve=DataInterface() {propertyValues {Curve {keys [{interpMode=RCIM_Cubic, tangentMode=RCTM_User, time=0.6, value=1}, {interpMode=RCIM_Cubic, tangentMode=RCTM_User, time=1, value=0}]}}})
    Scale RGB=&Lerp_Vector(
     A=&MakeVectorFromLinearColorRGB(Color=$System.ColorA)
     B=&MakeVectorFromLinearColorRGB(Color=$System.ColorB)
     Alpha=&FloatFromCurve(FloatCurve=DataInterface() {propertyValues {Curve {keys [{interpMode=RCIM_Cubic, tangentMode=RCTM_User, time=0.4, value=0}, {interpMode=RCIM_Cubic, tangentMode=RCTM_User, time=0.66, value=1}]}}})
    )
    ScaleA=true
    ScaleRGB=true
   )
   &DynamicMaterialParameters(
    Index 0 Param 2=&FloatFromCurve(FloatCurve=DataInterface() {propertyValues {Curve {keys [{arriveTangent=-1.9579, interpMode=RCIM_Cubic, time=0.2, value=1}, {interpMode=RCIM_Cubic, tangentMode=RCTM_User, time=0.45, value=0}]}}})
    Index 0 Param 4=&Lerp_Float(
     A=1
     B=&FloatFromCurve(FloatCurve=DataInterface() {propertyValues {Curve {keys [{tangentMode=RCTM_User, time=0.25, value=0.1}, {tangentMode=RCTM_User, time=1, value=1}]}}})
     Alpha=$System.Spray
    )
    Module.Index0Param0Write=true
    Module.Index0Param1Write=true
    Module.Index0Param2Write=true
    Module.Index0Param3Write=true
   )
  ]
  rendererClasses [&NiagaraSpriteRendererProperties]
  renderers [{rendererClass=&NiagaraSpriteRendererProperties}]
 }
 Empty {
  bEnabled=false
  emitterUpdateScript [
   &EmitterState()
   &SpawnBurst_Instantaneous(Use Spawn Probability=false, UseLoopCountLimit=true)
  ]
  particleSpawnScript [
   &InitializeParticle(
    Color Mode=::NewEnumerator0
    Lifetime=5
    Mesh Scale Mode=::NewEnumerator3
    Mesh Uniform Scale=3
    UsePositionOffset=false
    Write Material Random=true
   )
   &InitialMeshOrientation(Mesh Axis To Orient=<0 0 1>, Mesh Orientation Mode=::NewEnumerator2, Orientation Vector=<0 0 1>, Use Rotation Vector=false)
   /SetVariables(Particles.MeshOrientationSystem:Quat4f=&MakeQuatFromAxisAngle(Axis=<0 0 1>, AngleInDegrees=0, NormalizeAxis=true)) {isSetParameters=true}
  ]
  particleUpdateScript [
   &ParticleState(Kill Particles When Lifetime Has Elapsed=false, Let Infinitely Lived Particles Die When Emitter Deactivates=true)
   &UpdateMeshOrientation(
    Facing Direction=&TransformVector(Vector=$System.SprayDirection, Source Space=::World, Destination Space=::Simulation)
    Facing Interpolation=false
    Orientation Method=::NewEnumerator1
    Override Quat=false
    Stabilization Mode:ENiagara_OrientVectorCount=::NewEnumerator0
    TransformFacing=false
   )
  ]
  rendererClasses [&NiagaraMeshRendererProperties]
  renderers [
   {rendererClass=&NiagaraMeshRendererProperties, rendererIndex=0}
   {rendererClass=&NiagaraMeshRendererProperties, rendererIndex=1}
  ]
 }
 Foam_Spray {
  emitterUpdateScript [
   &EmitterState()
   &SpawnPerUnit(
    Max Movement Threshold=1000
    Module.Use Max Movement Threshold=true
    Module.Use Movement Tolerance=false
    Spawn Probability:NiagaraFloat=$System.Spray
    Spawn Spacing=15
    Use Spawn Probability=true
   )
   &SpawnRate(Spawn Probability=$System.Spray, SpawnRate=20, Use Spawn Probability=true)
  ]
  particleSpawnScript [
   &InitializeParticle(
    Lifetime Max=1.25
    Lifetime Mode=::NewEnumerator1
    Position Offset=&Multiply_VectorByFloat(
     Vector=&Lerp_Vector(A=$Engine.Owner.SystemZAxis, B=$System.SprayDirection, Alpha=0.3)
     Float=&Multiply_Float(A=&RandomRangeFloat(Minimum=70, Maximum=80), B=$System.Spray)
    )
    Position Offset Coordinate Space=::World
    Sprite Rotation Mode=::NewEnumerator2
    Sprite Size Mode=::NewEnumerator3
    Sprite UV Mode=::NewEnumerator3
    Uniform Sprite Size=&Lerp_Float(
     A=&RandomRangeFloat(Minimum=50, Maximum=60)
     B=&RandomRangeFloat(Minimum=160, Maximum=200)
     Alpha=$System.Spray
    )
    UsePositionOffset=true
    Write Material Random=true
   )
   &AddVelocity(
    Cone Angle=2
    Cone Axis=&Lerp_Vector(A=$Engine.Owner.SystemZAxis, B=$System.SprayDirection, Alpha=1)
    Module.UseConeFalloff=false
    Module.UseVelDistribution=false
    Override Seed=false
    Rotation Coordinate Space=::World
    Velocity Mode=::NewEnumerator2
    Velocity Speed=&Lerp_Float(A=100, B=600, Alpha=$System.Spray)
   ) {module=AddVelocity001}
   /SetVariables(Particles.SubImageIndex=&RandomRangeFloat(Minimum=0, Maximum=3)) {isSetParameters=true}
  ]
  particleUpdateScript [
   &ParticleState()
   &CameraOffset(Camera Offset Amount=&Multiply_Float(A=&MakeFloatFromVector2d(Vector 2D=$Particles.SpriteSize), B=0.33))
   &GravityForce(Gravity=&Multiply_VectorByFloat(Vector=<0 0 -980>, Float=1))
   &Drag(
    Drag=&RandomRangeFloat(Minimum=0.2, Maximum=0.4)
    Rotational Drag=50
    Use Linear Drag=true
    Use Rotational Drag=false
   )
   &InheritSourceMovement(
    Applied Position Scale=&VectorFromFloat(Value=&FloatFromCurve(
      FloatCurve=DataInterface() {propertyValues {Curve {keys [
          {arriveTangent=-5.1381, interpMode=RCIM_Cubic, leaveTangent=-5.1381, tangentMode=RCTM_User, time=0, value=1}
          {interpMode=RCIM_Cubic, tangentMode=RCTM_User, time=0.5, value=0}
         ]}}}
      Scale Curve=$System.Inherit Source Movement
     ))
    ApplyForce=false
    ApplyPosition=true
    ApplyPositionFalloff=false
   )
   &SolveForcesAndVelocity(Clamp Velocity=false, Limit Acceleration=false)
   &SpriteRotationRate(Rotation Rate=&FloatFromCurve(
     FloatCurve=DataInterface() {propertyValues {Curve {keys [<at 0 1 RCIM_Cubic>,<at 1 0.1>]}}}
     Scale Curve=&MakeCustomFloatFromBool(Bool=&RandomBool(), True Float=-90, False Float=90)
    ))
   &ScaleSpriteSize(Scale Factor=&Vector2DFromFloat(Value=&FloatFromCurve(FloatCurve=DataInterface() {propertyValues {Curve {keys [{arriveTangent=1.7782, time=0, value=0.05}, {arriveTangent=1.1318, time=0.15, value=0.25}, {leaveTangent=1.2213, time=1, value=1}]}}})))
   &ScaleColor(
    Scale Alpha=&FloatFromCurve(FloatCurve=DataInterface() {propertyValues {Curve {keys [
         <at 0 1 RCIM_Cubic>
         {interpMode=RCIM_Cubic, tangentMode=RCTM_User, time=0.1, value=1}
         {interpMode=RCIM_Cubic, tangentMode=RCTM_User, time=0.8, value=1}
         {arriveTangent=-6.292, interpMode=RCIM_Cubic, leaveTangent=-6.292, tangentMode=RCTM_User, time=1, value=0}
        ]}}})
    Scale RGB=&Lerp_Vector(
     A=&MakeVectorFromLinearColorRGB(Color=$System.ColorA)
     B=&MakeVectorFromLinearColorRGB(Color=$System.ColorB)
     Alpha=&FloatFromCurve(FloatCurve=DataInterface() {propertyValues {Curve {keys [{interpMode=RCIM_Cubic, tangentMode=RCTM_User, time=0, value=0}, {interpMode=RCIM_Cubic, tangentMode=RCTM_User, time=0.66, value=1}]}}})
    )
    ScaleA=false
    ScaleRGB=true
   )
   &DynamicMaterialParameters(
    Index 0 Param 2=&FloatFromCurve(
     FloatCurve=DataInterface() {propertyValues {Curve {keys [<at 0.1 1 RCIM_Cubic>, {interpMode=RCIM_Cubic, tangentMode=RCTM_User, time=0.3, value=0}]}}}
     Scale Curve=0
    )
    Index 0 Param 3=&Lerp_Float(
     A=1
     B=4
     Alpha=&FloatFromCurve(FloatCurve=DataInterface() {propertyValues {Curve {keys [<at 0 1 RCIM_Cubic>,<at 0.5 0 RCIM_Cubic>]}}})
    )
    Index 0 Param 4=&Lerp_Float(
     A=1
     B=&FloatFromCurve(FloatCurve=DataInterface() {propertyValues {Curve {keys [<at 0.05 0>, {interpMode=RCIM_Cubic, tangentMode=RCTM_User, time=0.4, value=0}, <at 1 1 RCIM_Cubic>]}}})
     Alpha=$System.Spray
    )
    Module.Index0Param0Write=true
    Module.Index0Param1Write=true
    Module.Index0Param2Write=true
    Module.Index0Param3Write=true
   )
   /SetVariables(Particles.ColorMist:LinearColor=&MultiplyLinearColor(Linear Color=$Particles.Color, Scale Factor=<1 1 1 0.1>)) {enabled=false, isSetParameters=true}
   &SpriteFacingAndAlignment(
    Sprite Alignment=&Lerp_Vector(
     A=&NormalizeVector(Vector To Normalize=$Engine.Owner.Velocity, Safe Normalize=true, Safe Fallback Threshold=0, Safe Fallback Vector=<0 0 0.001>)
     B=<0 0 0>
     Alpha=&FloatFromCurve(FloatCurve=DataInterface() {propertyValues {Curve {keys [{interpMode=RCIM_Cubic, tangentMode=RCTM_User, time=0, value=0}, {interpMode=RCIM_Cubic, tangentMode=RCTM_User, time=0.5, value=1}]}}})
    )
    WriteAlignment=true
    WriteFacing=false
   ) {enabled=false}
  ]
  rendererClasses [&NiagaraSpriteRendererProperties]
  renderers [{rendererClass=&NiagaraSpriteRendererProperties}]
 }
 Foam_Spray_Directional {
  emitterUpdateScript [
   &EmitterState()
   &SpawnBurst_Instantaneous(
    Spawn Count=10
    Spawn Time=&Multiply_Float(A=$User.Duration, B=0.75)
    Use Spawn Probability=false
    UseLoopCountLimit=true
   )
  ]
  particleSpawnScript [
   &InitializeParticle(
    Lifetime=0.33
    Position Offset=$System.PositionOffset
    Position Offset Coordinate Space=::Local
    Ribbon Width=40
    Ribbon Width Mode=::NewEnumerator1
    Sprite Size=<40 80>
    Sprite Size Mode=::NewEnumerator4
    Sprite UV Mode=::NewEnumerator1
    UsePositionOffset=true
    Write Material Random=true
   )
   /SetVariables(Particles.SprayDirection:Vector3f=$System.SprayDirection) {enabled=false, isSetParameters=true, module=SetVariables_#9}
  ]
  particleUpdateScript [
   &ParticleState(Kill Particles When Lifetime Has Elapsed=false, Let Infinitely Lived Particles Die When Emitter Deactivates=true)
   &KillParticles(Kill Particles=&SetBoolByFloatComparison(A=$StackContext.Timeline.NormalizedTime, B=0, Comparison Type=::NewEnumerator6)) {enabled=false}
   /SetVariables(Particles.Velocity=&TransformVector(
     Vector=&NormalizeVector(
      Vector To Normalize=&Lerp_Vector(
       A=$Engine.Owner.SystemZAxis
       B=$System.SprayDirection
       Alpha=&Power(a=&ReturnNormalizedExecIndex(), b=0.8)
      )
      Safe Normalize=false
      Safe Fallback Threshold=0
      Safe Fallback Vector=<0 0 0.001>
     )
     Source Space=::World
     Destination Space=::Local
    )) {isSetParameters=true, module=SetVariables_#10}
   /SetVariables(Particles.Position=&AddVectorToPosition(
     Position=$Particles.Initial.Position
     Vector=&Multiply_VectorByFloat(
      Vector=$Particles.Velocity
      Float=&FloatFromCurve(
       FloatCurve=DataInterface() {propertyValues {Curve {keys [{tangentMode=RCTM_User, time=0, value=0}, {tangentMode=RCTM_User, time=1, value=1}]}}}
       CurveIndex=&ReturnNormalizedExecIndex()
       Scale Curve=&Lerp_Float(A=60, B=160, Alpha=$System.Spray)
      )
     )
    )) {isSetParameters=true, module=SetVariables_#11}
   /SetVariables(Particles.RibbonU0Override=&Add_Float(
     A=&ReturnNormalizedExecIndex(Normalized Index Scale=0.5)
     B=&Multiply_Float(A=$Emitter.Age, B=-1.2)
    )) {isSetParameters=true, module=SetVariables_#12}
   &ScaleColor(
    Scale Alpha=&FloatFromCurve(
     FloatCurve=DataInterface() {propertyValues {Curve {keys [{interpMode=RCIM_Cubic, tangentMode=RCTM_User, time=0.75, value=1}, {interpMode=RCIM_Cubic, tangentMode=RCTM_User, time=1, value=0}]}}}
     CurveIndex=&ReturnNormalizedExecIndex()
     Scale Curve=$System.Spray
    )
    Scale RGB=&Lerp_Vector(
     A=&MakeVectorFromLinearColorRGB(Color=$System.ColorA)
     B=&MakeVectorFromLinearColorRGB(Color=$System.ColorB)
     Alpha=&FloatFromCurve(
      FloatCurve=DataInterface() {propertyValues {Curve {keys [{interpMode=RCIM_Cubic, tangentMode=RCTM_User, time=0, value=1}, {interpMode=RCIM_Cubic, tangentMode=RCTM_User, time=0.25, value=0}]}}}
      CurveIndex=&ReturnNormalizedExecIndex()
      Scale Curve=0.5
     )
    )
    ScaleA=true
    ScaleRGB=true
   )
   &DynamicMaterialParameters(
    Index 0 Param 2=&FloatFromCurve(
     FloatCurve=DataInterface() {propertyValues {Curve {keys [{arriveTangent=-1.9579, interpMode=RCIM_Cubic, time=0, value=1}, {interpMode=RCIM_Cubic, tangentMode=RCTM_User, time=0.8, value=0}]}}}
     CurveIndex=&ReturnNormalizedExecIndex()
    )
    Index 0 Param 4=&Lerp_Float(
     A=1
     B=&FloatFromCurve(
      FloatCurve=DataInterface() {propertyValues {Curve {keys [<at 0.5 0.1 RCIM_Cubic>,<at 1 1 RCIM_Cubic>]}}}
      CurveIndex=&ReturnNormalizedExecIndex()
     )
     Alpha=$System.Spray
    )
    Module.Param0WriteEnabled=false
    Module.Param1WriteEnabled=true
    Module.Param2WriteEnabled=false
    Module.Param3WriteEnabled=true
   )
   &ScaleRibbonWidth(Ribbon Width Scale=&FloatFromCurve(
     FloatCurve=DataInterface() {propertyValues {Curve {keys [{arriveTangent=0.9448, interpMode=RCIM_Cubic, time=0, value=0.4}, {interpMode=RCIM_Cubic, leaveTangent=4.7864, time=1, value=1}]}}}
     CurveIndex=&ReturnNormalizedExecIndex()
    ))
  ]
  rendererClasses [&NiagaraRibbonRendererProperties]
  renderers [{rendererClass=&NiagaraRibbonRendererProperties}]
 }
 Foam_Spray_Directional001 {
  bEnabled=false
  emitterUpdateScript [
   &EmitterState()
   &SpawnRate(Spawn Probability=$System.Spray, SpawnRate=10, Use Spawn Probability=true)
  ]
  particleSpawnScript [
   &InitializeParticle(
    Lifetime=0.33
    Position Offset=$System.PositionOffset
    Position Offset Coordinate Space=::Local
    Ribbon Width=30
    Ribbon Width Mode=::NewEnumerator1
    Sprite Size=<40 80>
    Sprite Size Mode=::NewEnumerator4
    Sprite UV Mode=::NewEnumerator1
    UsePositionOffset=true
    Write Material Random=true
   )
   /SetVariables(Particles.SubImageIndex=&SelectFloatFromArray(
     Array Sampling Mode=::NewEnumerator0
     Direct Array Index=?
     Float Selection Array=DataInterface() {propertyValues {FloatData [2,2,3,0,3]}}
     Random Seed=0
    )) {isSetParameters=true}
   &SubUVAnimation(End Frame=3, SubUV Animation Mode:ENiagara_SubUVLookupMode=::NewEnumerator2) {enabled=false}
   /SetVariables(Particles.SprayDirection:Vector3f=$System.SprayDirection) {enabled=false, isSetParameters=true, module=SetVariables_#9}
  ]
  particleUpdateScript [
   &ParticleState()
   /SetVariables(Particles.Velocity=&TransformVector(
     Vector=&NormalizeVector(
      Vector To Normalize=&Lerp_Vector(A=$Engine.Owner.SystemZAxis, B=$System.SprayDirection, Alpha=&Power(a=$Particles.NormalizedAge, b=1.5))
      Safe Normalize=false
      Safe Fallback Threshold=0
      Safe Fallback Vector=<0 0 0.001>
     )
     Source Space=::World
     Destination Space=::Local
    )) {isSetParameters=true, module=SetVariables_#10}
   /SetVariables(Particles.Position=&AddVectorToPosition(
     Position=$Particles.Initial.Position
     Vector=&Multiply_VectorByFloat(
      Vector=$Particles.Velocity
      Float=&FloatFromCurve(
       FloatCurve=DataInterface() {propertyValues {Curve {keys [{tangentMode=RCTM_User, time=0, value=0}, {tangentMode=RCTM_User, time=1, value=1}]}}}
       Scale Curve=&Lerp_Float(A=10, B=120, Alpha=$System.Spray)
      )
     )
    )) {isSetParameters=true, module=SetVariables_#11}
   /SetVariables(Particles.RibbonU0Override=&Add_Float(A=$Particles.NormalizedAge, B=&Multiply_Float(A=$Emitter.Age, B=0.1))) {isSetParameters=true, module=SetVariables_#12}
   &ScaleSpriteSize(
    Scale Factor=&MakeVector2D(
     X=&FloatFromCurve(FloatCurve=DataInterface() {propertyValues {Curve {keys [{arriveTangent=1.6254, interpMode=RCIM_Cubic, leaveTangent=1.6254, tangentMode=RCTM_User, time=0, value=0.4}, <at 1 1 RCIM_Cubic>]}}})
     Y=&FloatFromCurve(FloatCurve=DataInterface() {propertyValues {Curve {keys [
          {arriveTangent=4.1785, interpMode=RCIM_Cubic, leaveTangent=4.1785, tangentMode=RCTM_User, time=0, value=0}
          {arriveTangent=1.1179, interpMode=RCIM_Cubic, leaveTangent=1.1179, tangentMode=RCTM_User, time=0.2042, value=0.6228}
          {arriveTangent=0.0224, interpMode=RCIM_Cubic, leaveTangent=0.0224, tangentMode=RCTM_User, time=1, value=1}
         ]}}})
    )
    Scale Sprite Size Mode=::NewEnumerator2
   )
   &ScaleColor(
    Scale Alpha=&FloatFromCurve(FloatCurve=DataInterface() {propertyValues {Curve {keys [{interpMode=RCIM_Cubic, tangentMode=RCTM_User, time=0.6, value=1}, {interpMode=RCIM_Cubic, tangentMode=RCTM_User, time=1, value=0}]}}})
    Scale RGB=&Lerp_Vector(
     A=&MakeVectorFromLinearColorRGB(Color=$System.ColorA)
     B=&MakeVectorFromLinearColorRGB(Color=$System.ColorB)
     Alpha=&FloatFromCurve(
      FloatCurve=DataInterface() {propertyValues {Curve {keys [{interpMode=RCIM_Cubic, tangentMode=RCTM_User, time=0, value=1}, {interpMode=RCIM_Cubic, tangentMode=RCTM_User, time=0.25, value=0}]}}}
      Scale Curve=0.5
     )
    )
    ScaleA=true
    ScaleRGB=true
   )
   &DynamicMaterialParameters(
    Index 0 Param 2=&FloatFromCurve(FloatCurve=DataInterface() {propertyValues {Curve {keys [{arriveTangent=-1.9579, interpMode=RCIM_Cubic, time=0, value=1}, {interpMode=RCIM_Cubic, tangentMode=RCTM_User, time=0.75, value=0}]}}})
    Index 0 Param 4=&FloatFromCurve(
     FloatCurve=DataInterface() {propertyValues {Curve {keys [
         <at 0.5 0.1 RCIM_Cubic>
         {arriveTangent=1.8, interpMode=RCIM_Cubic, leaveTangent=1.8, time=0.8, value=0.8}
         {tangentMode=RCTM_User, time=1, value=1}
        ]}}}
     Scale Curve=$System.SprayErosion
    )
    Module.Param0WriteEnabled=false
    Module.Param1WriteEnabled=true
    Module.Param2WriteEnabled=false
    Module.Param3WriteEnabled=true
   )
  ]
  rendererClasses [&NiagaraSpriteRendererProperties]
  renderers [{rendererClass=&NiagaraSpriteRendererProperties}]
 }
}
summary {
 systemName=NS_Paprika_TeamSpray_Spray
 emitters [
  {emitterName=Foam_Spray, rendererClasses [&NiagaraSpriteRendererProperties]}
  {emitterName=Dripping_Foam, rendererClasses [&NiagaraSpriteRendererProperties]}
  {bEnabled=false, emitterName=Foam_Spray_Directional001, rendererClasses [&NiagaraSpriteRendererProperties]}
  {bEnabled=false, emitterName=Empty, rendererClasses [&NiagaraMeshRendererProperties]}
  {emitterName=Foam_Spray_Directional, rendererClasses [&NiagaraRibbonRendererProperties]}
 ]
 userVariables [
  User.Duration {type=NiagaraFloat, defaultValue=2.4, description=""}
  User.FillPercent {type=NiagaraFloat, defaultValue=1, description=""}
  User.IsFinished {type=NiagaraBool}
  User.IsSpraying {type=NiagaraBool, defaultValue=true}
  User.LiquidScale {type=NiagaraFloat, defaultValue=1, description=""}
 ]
}
