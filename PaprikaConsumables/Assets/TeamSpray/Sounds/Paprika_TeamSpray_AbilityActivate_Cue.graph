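# Sound cue for the activation sound of the Paprika Team Spray consumable ability.
# A random node picks one of three equally weighted wave variants on each play.
# It plays at reduced volume through the consumables sound class, with consumable-use attenuation limiting its range.
# A concurrency override caps simultaneous instances and stops the oldest when the limit is exceeded.
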SoundCue /PaprikaConsumables/Assets/TeamSpray/Sounds/Paprika_TeamSpray_AbilityActivate_Cue
root=Random_0
nodes [
 {id=Random_0, children [WavePlayer_0,WavePlayer_2,WavePlayer_3], properties {weights [1,1,1]}}
 {id=WavePlayer_0, properties {soundWaveAssetPtr=/PaprikaConsumables/Assets/TeamSpray/Sounds/Waves/Paprika_TeamSpray_AbilityActivate_01}}
 {id=WavePlayer_2, properties {soundWaveAssetPtr=/PaprikaConsumables/Assets/TeamSpray/Sounds/Waves/Paprika_TeamSpray_AbilityActivate_02}}
 {id=WavePlayer_3, properties {soundWaveAssetPtr=/PaprikaConsumables/Assets/TeamSpray/Sounds/Waves/Paprika_TeamSpray_AbilityActivate_03}}
]
properties {
 attenuationSettings=&Fort_Consumable_Use_Athena_Att
 bOverrideConcurrency=true
 maxDistance=2250
 soundClassObject=&Fort_Consumables
 volumeMultiplier=0.7
 concurrencyOverrides {maxCount=4, resolutionRule=StopOldest}
}
