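# Sound cue for the looping ability sound of the Paprika TeamSpray consumable.
# A single looping wave player feeds the root directly, playing the AbilityLoop wave at slightly reduced volume.
# It routes through the Fort consumables sound class and consumable-use attenuation.
# A concurrency override caps simultaneous instances at a small count, stopping the oldest first.
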SoundCue /PaprikaConsumables/Assets/TeamSpray/Sounds/Paprika_TeamSpray_AbilityLoop_Cue
root=WavePlayer_0
nodes [{
  id=WavePlayer_0
  properties {bLooping=true, soundWaveAssetPtr=/PaprikaConsumables/Assets/TeamSpray/Sounds/Waves/Paprika_TeamSpray_AbilityLoop}
 }]
properties {
 attenuationSettings=&Fort_Consumable_Use_Athena_Att
 bOverrideConcurrency=true
 maxDistance=2250
 soundClassObject=&Fort_Consumables
 volumeMultiplier=0.7
 concurrencyOverrides {maxCount=4, resolutionRule=StopOldest}
}
