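# Sound cue for the cloth rustle heard during the Team Spray consumable ability in the Paprika consumables plugin.
# A Random node picks one of six ClothRust wave variations with equal weight on each play.
# It plays at reduced volume through consumable-use attenuation and the consumables sound class, with its own concurrency override that stops the oldest instance when the small voice limit is exceeded.
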SoundCue /PaprikaConsumables/Assets/TeamSpray/Sounds/Paprika_TeamSpray_Ability_ClothRust_Cue
root=Random_0
nodes [
 {id=Random_0, children [WavePlayer_0,WavePlayer_2,WavePlayer_3,WavePlayer_1,WavePlayer_4,WavePlayer_5], properties {weights [1,1,1,1,1,1]}}
 {id=WavePlayer_0, properties {soundWaveAssetPtr=/PaprikaConsumables/Assets/TeamSpray/Sounds/Waves/Paprika_TeamSpray_Ability_ClothRust_01}}
 {id=WavePlayer_1, properties {soundWaveAssetPtr=/PaprikaConsumables/Assets/TeamSpray/Sounds/Waves/Paprika_TeamSpray_Ability_ClothRust_04}}
 {id=WavePlayer_2, properties {soundWaveAssetPtr=/PaprikaConsumables/Assets/TeamSpray/Sounds/Waves/Paprika_TeamSpray_Ability_ClothRust_02}}
 {id=WavePlayer_3, properties {soundWaveAssetPtr=/PaprikaConsumables/Assets/TeamSpray/Sounds/Waves/Paprika_TeamSpray_Ability_ClothRust_03}}
 {id=WavePlayer_4, properties {soundWaveAssetPtr=/PaprikaConsumables/Assets/TeamSpray/Sounds/Waves/Paprika_TeamSpray_Ability_ClothRust_05}}
 {id=WavePlayer_5, properties {soundWaveAssetPtr=/PaprikaConsumables/Assets/TeamSpray/Sounds/Waves/Paprika_TeamSpray_Ability_ClothRust_06}}
]
properties {
 attenuationSettings=&Fort_Consumable_Use_Athena_Att
 bOverrideConcurrency=true
 maxDistance=2250
 soundClassObject=&Fort_Consumables
 volumeMultiplier=0.4
 concurrencyOverrides {maxCount=4, resolutionRule=StopOldest}
}
