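# Sound cue for closing the bottle on the TeamSpray consumable in the PaprikaConsumables plugin.
# A Random node picks between two equally weighted close-bottle wave variations.
# It plays in the Fort consumables sound class with consumable-use attenuation and a boosted volume.
# A concurrency override caps simultaneous instances, stopping the oldest when the limit is reached.
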SoundCue /PaprikaConsumables/Assets/TeamSpray/Sounds/Paprika_TeamSpray_CloseBottle_Cue
root=Random_0
nodes [
 {id=Random_0, children [WavePlayer_1,WavePlayer_2], properties {weights [1,1]}}
 {id=WavePlayer_1, properties {soundWaveAssetPtr=/PaprikaConsumables/Assets/TeamSpray/Sounds/Waves/Paprika_TeamSpray_CloseBottle_01}}
 {id=WavePlayer_2, properties {soundWaveAssetPtr=/PaprikaConsumables/Assets/TeamSpray/Sounds/Waves/Paprika_TeamSpray_CloseBottle_02}}
]
properties {
 attenuationSettings=&Fort_Consumable_Use_Athena_Att
 bOverrideConcurrency=true
 maxDistance=2250
 soundClassObject=&Fort_Consumables
 volumeMultiplier=2.25
 concurrencyOverrides {maxCount=4, resolutionRule=StopOldest}
}
