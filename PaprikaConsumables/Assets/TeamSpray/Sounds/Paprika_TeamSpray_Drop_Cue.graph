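# Sound cue for dropping the Paprika Team Spray consumable.
# A Random node picks one of four equally weighted drop movement waves to vary each play.
# It plays at reduced volume through the Fort_Loot_Collect sound class, spatialized with the shared weapon drop attenuation.
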SoundCue /PaprikaConsumables/Assets/TeamSpray/Sounds/Paprika_TeamSpray_Drop_Cue
root=Random_0
nodes [
 {id=Random_0, children [WavePlayer_0,WavePlayer_4,WavePlayer_5,WavePlayer_6], properties {weights [1,1,1,1]}}
 {id=WavePlayer_0, properties {soundWaveAssetPtr=/PaprikaConsumables/Assets/TeamSpray/Sounds/Waves/Paprika_TeamSpray_Drop_Mvmt_01}}
 {id=WavePlayer_4, properties {soundWaveAssetPtr=/PaprikaConsumables/Assets/TeamSpray/Sounds/Waves/Paprika_TeamSpray_Drop_Mvmt_02}}
 {id=WavePlayer_5, properties {soundWaveAssetPtr=/PaprikaConsumables/Assets/TeamSpray/Sounds/Waves/Paprika_TeamSpray_Drop_Mvmt_03}}
 {id=WavePlayer_6, properties {soundWaveAssetPtr=/PaprikaConsumables/Assets/TeamSpray/Sounds/Waves/Paprika_TeamSpray_Drop_Mvmt_04}}
]
properties {
 attenuationSettings=&Fort_Weapon_Drop_Att
 maxDistance=3000
 soundClassObject=&Fort_Loot_Collect
 volumeMultiplier=0.4
}
