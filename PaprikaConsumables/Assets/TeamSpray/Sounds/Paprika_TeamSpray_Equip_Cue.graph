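# Sound cue for the equip sound of the TeamSpray consumable in PaprikaConsumables.
# A Random node picks one of four equip wave variations with equal weight for variety.
# It plays at reduced volume through the gadgets sound class, using the shared weapon reload attenuation for spatial falloff.
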SoundCue /PaprikaConsumables/Assets/TeamSpray/Sounds/Paprika_TeamSpray_Equip_Cue
root=Random_2
nodes [
 {id=Random_2, children [WavePlayer_8,WavePlayer_9,WavePlayer_0,WavePlayer_1], properties {weights [1,1,1,1]}}
 {id=WavePlayer_0, properties {soundWaveAssetPtr=/PaprikaConsumables/Assets/TeamSpray/Sounds/Waves/Paprika_TeamSpray_Equip_03}}
 {id=WavePlayer_1, properties {soundWaveAssetPtr=/PaprikaConsumables/Assets/TeamSpray/Sounds/Waves/Paprika_TeamSpray_Equip_04}}
 {id=WavePlayer_8, properties {soundWaveAssetPtr=/PaprikaConsumables/Assets/TeamSpray/Sounds/Waves/Paprika_TeamSpray_Equip_01}}
 {id=WavePlayer_9, properties {soundWaveAssetPtr=/PaprikaConsumables/Assets/TeamSpray/Sounds/Waves/Paprika_TeamSpray_Equip_02}}
]
properties {
 attenuationSettings=&Fort_Weapon_Reload_Att
 maxDistance=2250
 soundClassObject=&Fort_Gadgets
 volumeMultiplier=0.4
}
